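; Timed Niagara anim notify for Pulse Lancer effects that detaches the spawned effect's location and rotation from the mesh's movement.
; On notify begin it runs the parent spawn logic, then casts the spawned effect to a Niagara component.
; It caches the world transform, makes location and rotation absolute (scale stays relative), and reapplies the cached values to hold the effect in place.
; It also passes the mesh's right vector to the Niagara "Direction" variable.
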(fn Received_NotifyBegin
  (params (MeshComp :type Skeletal%20Mesh%20Component%20Object%20Reference) (Animation :type Anim%20Sequence%20Base%20Object%20Reference) (TotalDuration :type Float%20%28single-precision%29) (EventReference :type Anim%20Notify%20Event%20Reference%20Structure%20%28by%20ref%29))
  (bind _returnvalue (|CallParentFunction MeshComp Animation TotalDuration EventReference))
  (bind _returnvalue_2 (AnimNotify|GetSpawnedEffect MeshComp))
  (bind _asniagara_particle_system_component (Utilities|Casting|CastToNiagaraComponent _returnvalue_2)
    (:then
      (bind _output_get (LocalVariables|SetTmpRotation (Transformation|GetWorldRotation _asniagara_particle_system_component)))
      (bind _output_get_2 (LocalVariables|SetTmpLocation (Transformation|GetWorldLocation _asniagara_particle_system_component)))
      (Transformation|SetAbsolute _returnvalue_2 true true false)
      (Transformation|SetWorldLocationAndRotation _asniagara_particle_system_component _output_get_2 _output_get false false)
      (Niagara|SetNiagaraVariable(Vector3) _asniagara_particle_system_component "Direction" (Transformation|GetRightVector MeshComp)))
    (:CastFailed
      _))
  (return _returnvalue))
