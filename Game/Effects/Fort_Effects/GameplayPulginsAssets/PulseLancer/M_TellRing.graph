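# Unlit translucent Niagara material for PulseLancer that draws a tell ring, using WBT decal projection by default or radial mesh UVs when switched off.
# Noise-distorted radial UVs drive a ring smoke texture and a ribbon trail texture, tinted between two colors with darkened-side shading.
# Dynamic parameters supply offset, radial scale, time, ribbon intensity and decal bounds.
# Decal depth masks limit opacity, with depth testing disabled.
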Material /Game/Effects/Fort_Effects/GameplayPulginsAssets/PulseLancer/M_TellRing
graph {
 properties {
  bAutomaticallySetUsageInEditor=false
  bDisableDepthTest=true
  bUsedWithNiagaraMeshParticles=true
  bUsedWithNiagaraRibbons=true
  bUsedWithNiagaraSprites=true
  blendMode=BLEND_Translucent
  shadingModel=MSM_Unlit
  usedShadingModels=Unlit
 }
 nodes [
  n1 = VectorParameter() {parameterName=C2, defaultValue=<rgba 3.4435 0.1722 0.4448 1>}
  n2 = TextureObjectParameter() {
   group=Noise2
   parameterName=Noise2_Texture
   samplerType=SAMPLERTYPE_Masks
   sortPriority=0
   texture=/Game/Weapons/WeaponSkins/Wraps/Textures/Mash/Mash_Noise
  }
  n3 = VectorParameter() {group=Noise2, parameterName=Noise2_UVControl, channelNames {a=VPan, b=UPan, g=VScale, r=UScale}, defaultValue=<rgba 1 1 0.03 0>}
  n4 = DynamicParameter() {parameterIndex=3, defaultValue=<rgba 0 1000 1 1>, paramNames [BoundX,BoundY,BoundZ,Param4]}
  n5 = MaterialFunctionCall!(MeshOriginalBoundSize=MaterialFunctionCall(X=@n4.BoundX, Y=@n4.BoundX, Z=@n4.BoundX) {materialFunction=&MakeFloat3}.Result) {materialFunction=/Game/Effects/Niagara/MasterMaterials/Functions/Common/FXCommon_WBTDecalUV}
  n6 = StaticSwitchParameter(
   True=@n5.RadialUV
   False=MaterialFunctionCall() {materialFunction=&VectorToRadialValue}.Radial Coordinates
  ) {defaultValue=true, parameterName=USEWBTDecal}
  n7 = VectorParameter() {
   group=Noise2
   parameterName=Noise2_NoiseControl
   sortPriority=60
   channelNames {a=VStrength, b=UStrength, g=VOffset, r=UOffset}
   defaultValue=<rgba -0.5 -0.5 0 0.0936>
  }
  n8 = MaterialFunctionCall!(
   NoiseValue=MaterialFunctionCall(Texture=@n2, UVControl_RGB=@n3, UVControl_A=@n3, "[Override]UVInput"=@n6) {materialFunction=&FXCommon_TextureSample}.RG
   Control_RGB=@n7
   Control_A=@n7
  ) {materialFunction=/Game/Effects/Niagara/MasterMaterials/Functions/Common/FXCommon_NoiseControl}
  n9 = TextureObjectParameter() {
   group=Noise
   parameterName=Noise_Texture
   samplerType=SAMPLERTYPE_Normal
   sortPriority=0
   texture=/Game/Characters/Player/Female/Medium/Bodies/F_MED_Bistro_Astronaut/FX/T_GaussianNoise_Normal
  }
  n10 = VectorParameter() {group=Noise, parameterName=Noise_UVControl, channelNames {a=VPan, b=UPan, g=VScale, r=UScale}, defaultValue=<rgba 1 1.5 0.3 0>}
  n11 = VectorParameter() {
   group=Noise
   parameterName=Noise_NoiseControl
   sortPriority=60
   channelNames {a=VStrength, b=UStrength, g=VOffset, r=UOffset}
   defaultValue=<rgba -0.5 -0.5 0 0.05>
  }
  n12 = MaterialFunctionCall(
   NoiseValue=MaterialFunctionCall(Texture=@n9, UVControl_RGB=@n10, UVControl_A=@n10, "[Override]UVInput"=@n6) {materialFunction=&FXCommon_TextureSample}.RG
   Control_RGB=@n11
   Control_A=@n11
  ) {materialFunction=&FXCommon_NoiseControl}
  n13 = DynamicParameter() {defaultValue=<rgba 0 1000 1 1>, paramNames [Offset,SpriteSize,Ribbon2Color,Time]}
  n14 = AppendVector(A=Constant(), B=Add(A=@n13.Offset, B=ScalarParameter() {parameterName=VOffset}))
  n15 = VectorParameter() {group=Ribbon2, parameterName=Ribbon2_UVControl, channelNames {a=VPan, b=UPan, g=VScale, r=UScale}, defaultValue=<rgba 1 3 0.025 0>}
  n16 = NamedRerouteUsage()
  n17 = StaticBool() {value=true}
  n18 = MaterialFunctionCall!(
   Texture
   AdditionalUVOffset=Add(A=Add(A=@n8.NoiseValue, B=@n12.NoiseValue), B=@n14)
   UVControl_RGB=Multiply(A=@n15, B=NamedRerouteUsage())
   UVControl_A=@n15.VPan
   "[Override]TimeInput"=@n16
   "[Override]ScaleUVByCenter"=@n17
   "[Override]UVInput"=@n6
  ) {materialFunction=/Game/Effects/Niagara/MasterMaterials/Functions/Common/FXCommon_TextureSample_V2}
  n19 = TextureObjectParameter() {
   group=Ribbon2
   parameterName=Ribbon2_Texture
   sortPriority=0
   texture=/Game/Weapons/FORT_Melee/Pickaxe_Gimmick_Female/FX/T_Pickaxe_Gimmick_Female_Trail_01
  }
  n20 = NamedRerouteUsage()
  n21 = DDX(Value=@n20)
  n22 = DDY(Value=@n20)
  n23 = TextureSample!(UVs=@n18.UV, Tex=@n19, "DDX(UVs)"=@n21, "DDY(UVs)"=@n22, Apply View MipBias) {
   mipValueMode=TMVM_Derivative
   texture=/Game/Weapons/FORT_Melee/Pickaxe_Future_Samurai_Summer_Male/FX/T_Pickaxe_Future_Samurai_Summer_Male_Trail_01
  }
  n24 = Multiply(
   A=MaterialFunctionCall!(TexCoord=@n18.UV, ClampChannel=Constant2Vector() {g=1}) {materialFunction=/Game/Effects/Fort_Effects/MaterialFunctions/MF_FX_TilingMethod}.Result
   B=@n23.R
  )
  n25 = VectorParameter() {parameterName=Ribbon2_RefineHighLight, defaultValue=<rgba 1.6036 100 8 0.5>}
  n26 = Multiply(
   A=MaterialFunctionCall(Texture=@n24, RGB=@n25.RGB, A=@n25.RGB) {materialFunction=&MF_FX_Refine}.Result
   B=@n13.Ribbon2Color
  )
  n27 = VectorParameter() {parameterName=C1, defaultValue=<rgba 0.5287 0 0.0881 1>}
  n28 = VectorParameter() {group=RingTexture, parameterName=RingTexture_UVControl, channelNames {a=VPan, b=UPan, g=VScale, r=UScale}, defaultValue=<rgba 1 4 0.0308 0>}
  n29 = MaterialFunctionCall(
   AdditionalUVOffset=Add(A=@n14, B=@n12.NoiseValue)
   UVControl_RGB=Multiply(A=@n28, B=NamedRerouteUsage())
   UVControl_A=@n28.VPan
   "[Override]TimeInput"=@n16
   "[Override]ScaleUVByCenter"=@n17
   "[Override]UVInput"=@n6
  ) {materialFunction=&FXCommon_TextureSample_V2}
  n30 = TextureObjectParameter() {group=RingTexture, parameterName=RingTexture_Texture, sortPriority=0, texture=/Game/Effects/Fort_Effects/Textures/Smoke/T_Smoke_Trail_Soft_2}
  n31 = TextureSample(UVs=@n29.UV, Tex=@n30, "DDX(UVs)"=@n21, "DDY(UVs)"=@n22) {
   mipValueMode=TMVM_Derivative
   texture=&T_Pickaxe_Future_Samurai_Summer_Male_Trail_01
  }
  n32 = Multiply(
   A=MaterialFunctionCall(TexCoord=@n29.UV, ClampChannel=Constant2Vector() {g=1}) {materialFunction=&MF_FX_TilingMethod}.Result
   B=Desaturation(None=@n31.RGB, Fraction=Constant() {r=1})
  )
  n33 = VectorParameter() {parameterName=RingTexture_RefineHighLight, defaultValue=<rgba 7.6462 12.292 0.1 0.5>}
  n34 = MaterialFunctionCall(Texture=@n32, RGB=@n33.RGB, A=@n33.RGB) {materialFunction=&MF_FX_Refine}
  n35 = Add(
   A=OneMinus(None=MaterialFunctionCall(Float2=@n29.UV) {materialFunction=&BreakOutFloat2Components}.G)
   B=ScalarParameter() {defaultValue=0.0512, parameterName=FE}
  )
  n36 = ScalarParameter() {defaultValue=3, parameterName=CC}
  n37 = Multiply(
   A=Add(A=Multiply(A=@n1.RGB, B=@n26), B=LinearInterpolate(A=@n27.RGB, B=@n1.RGB, Alpha=@n34.Result))
   B=LinearInterpolate(
    A=ScalarParameter() {defaultValue=0.4, parameterName=DarkSide}
    Alpha=MaterialFunctionCall(In=@n35, Contrast=@n36) {materialFunction=&CheapContrast}.Result
   )
  )
  n38 = ParticleColor()
  MaterialFunctionCall(Emissive=Divide(A=Multiply(A=@n37, B=@n38.RGB)) {constB=2.271}) {materialFunction=&MF_FX_EmissiveAdjustments}
  n39 = VectorParameter() {parameterName=Ribbon2_RefineOpacity, defaultValue=<rgba 1 0 0.1024 1>}
  n40 = Multiply(
   A=MaterialFunctionCall(TexCoord=@n18.UV, ClampChannel=Constant2Vector() {g=1}) {materialFunction=&MF_FX_TilingMethod}.Result
   B=MaterialFunctionCall(Texture=@n24, RGB=@n39.RGB, A=@n39.RGB) {materialFunction=&MF_FX_Refine}.Result
  )
  n41 = VectorParameter() {parameterName=RingTexture_Refine, defaultValue=<rgba 0.5 2.5 1 0.5>}
  n42 = MaterialFunctionCall(Texture=@n32, RGB=@n41.RGB, A=@n41.RGB) {materialFunction=&MF_FX_Refine}
  n43 = NamedRerouteUsage()
  n44 = NamedRerouteUsage()
  n45 = ScalarParameter() {group=ZMask, parameterName=ZMask_GrayAdjust_Add}
  n46 = Saturate(None=Multiply(A=Add(A=@n44, B=@n45), B=ScalarParameter() {defaultValue=1, group=ZMask, parameterName=ZMask_GrayAdjust_Mul}))
  n47 = ScalarParameter() {defaultValue=1, group=ZMask, parameterName=ZMask_GrayAdjust_Power}
  n48 = Multiply(A=NamedRerouteUsage(), B=Saturate(None=Power(Base=@n46, Exp=@n47)))
  Multiply(
   A=Multiply(A=@n38.A, B=Add(A=@n40, B=@n42.Result))
   B=StaticSwitchParameter(True=Multiply(A=@n43, B=@n48), False=Constant() {r=1}) {defaultValue=true, parameterName=USEWBTDecal}
  )
  n49 = Constant() {r=1}
  NamedRerouteDeclaration(None=MaterialFunctionCall(X=@n49, Y=Divide(A=@n13.SpriteSize) {constB=1000}, Z=@n49) {materialFunction=&MakeFloat3}.Result) {name=ScaleUV, nodeColor=<rgba 0 0.6016 1 1>}
  NamedRerouteDeclaration(None=@n13.Time) {name=TimeIn, nodeColor=<rgba 1 0.9141 0 1>}
  NamedRerouteDeclaration(None=@n6) {name=RadialUVOut, nodeColor=<rgba 0 1 0.2031 1>}
  NamedRerouteDeclaration(None=StaticSwitchParameter(True=@n5.UV, False=TextureCoordinate()) {defaultValue=true, parameterName=USEWBTDecal}) {name=UVOut, nodeColor=<rgba 1 0 0.1172 1>}
  NamedRerouteDeclaration(None=@n5."CustomDepthMask(1onMobile)") {name=CustomDepthMask, nodeColor=<rgba 1 0 0.2812 1>}
  NamedRerouteDeclaration(None=@n5.ZMask) {name=ZMask, nodeColor=<rgba 0 0.2031 1 1>}
  NamedRerouteDeclaration(None=@n5.DepthMask) {name=DepthMask, nodeColor=<rgba 0.4297 0 1 1>}
  NamedRerouteUsage()
 ]
}
